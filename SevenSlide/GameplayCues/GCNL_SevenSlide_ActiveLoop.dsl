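; Looping gameplay cue for SevenSlide boots: attaches idle VFX to the ExtraPart mesh and starts a looping sound indicator for pawns off the local team.
; A fast timer runs IsSlidingCheck, feeding particle parameters from sliding and water-walking tags, normalized fuel attributes and the HangTime tag.
; Start also spawns an attached wash-force actor with timeline-driven strength.
; ZValueTick, a foot-trace debug event, is not bound to that timer.
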(event Custom|IsSlidingCheck
  (bind _fuelattribute (Variables|Default|GetFuelAttribute))
  (bind _as_fort_player_pawn (Variables|Default|GetAs%20Fort%20Player%20Pawn))
  (bind _returnvalue (Game|GetPlayerController _as_fort_player_pawn))
  (bind (_bsuccessfullyfoundattribute _returnvalue_2) (Ability|Attribute|GetFloatAttribute _returnvalue _fuelattribute))
  (bind _maxfuelattribute (Variables|Default|GetMaxFuelAttribute))
  (bind (_bsuccessfullyfoundattribute_2 _returnvalue_3) (Ability|Attribute|GetFloatAttribute _returnvalue _maxfuelattribute))
  (bind _particlecomps (Variables|Default|GetParticleComps))
  (bind _output (Utilities|Array|Get(acopy) _particlecomps 1))
  ; SLIDING AND WATER WALKING CHECK
  (Effects|Components|ParticleSystem|SetBoolParameter (Utilities|Array|Get(acopy) (Variables|Default|GetParticleComps)) "IsSliding" (and (Widget|UEFN|InvertBool (GameplayTags|HasMatchingGameplayTag (Variables|Default|GetAs%20Fort%20Player%20Pawn) "(TagName=\"Athena.Gameplay.Status.WaterWalking\")")) (GameplayTags|HasMatchingGameplayTag (Variables|Default|GetAs%20Fort%20Player%20Pawn) "(TagName=\"MovementMode.Sliding\")")))
  ; OVERHEATING STATE
  (if (and _bsuccessfullyfoundattribute _bsuccessfullyfoundattribute_2)
    (Effects|Components|ParticleSystem|SetFloatParameter _output "Fuel" (Utilities|RemapFloat _returnvalue_2 (Math|Vector2D|MakeVector2D 0.0 _returnvalue_3) "(X=0.000000,Y=1.000000)" true))
    (else
      _))
  (Effects|Components|ParticleSystem|SetBoolParameter _output "IsSlomo" (GameplayTags|HasMatchingGameplayTag (Variables|Default|GetAs%20Fort%20Player%20Pawn) "(TagName=\"sevenslide.HangTime\")")))

(event EventOnLoopingStart(EitherVFX-preferthisversion)
  (params (MyTarget :type Actor%20Object%20Reference) (Parameters :type Gameplay%20Cue%20Parameters%20Structure%20%28by%20ref%29) (ParticleComponents :type Array%20of%20FXSystem%20Component%20Object%20References) (AudioComponents :type Array%20of%20Audio%20Component%20Object%20References))
  (|CallParentFunction MyTarget Parameters ParticleComponents AudioComponents)
  (Variables|Default|SetParticleComps ParticleComponents)
  (bind _asfort_player_pawn (Utilities|Casting|CastToFortPlayerPawn MyTarget))
  (bind _output_get (Variables|Default|SetAs%20Fort%20Player%20Pawn _asfort_player_pawn))
  (if (not (Team|OnLocalPlayersTeam _output_get _output_get))
    (SoundIndicator|StartTargetedSoundIndicatorLooping (Variables|Default|GetFortSoundIndicator) _output_get "/SevenSlide/Assets/Audio/MSS_SevenSlide_Movement_Loop.MSS_SevenSlide_Movement_Loop" 0.1 (Variables|/Script/Engine.Pawn|GetPlayerState _output_get)))
  ; ATTACH IDLE FX TO THE BOOTS COMPONENT
  (Transformation|AttachComponentToComponent (Utilities|Array|Get(acopy) ParticleComponents 1) (Customization|GetSkeletalMeshforPartType (Variables|Default|GetAs%20Fort%20Player%20Pawn) "ExtraPart" :!class /Script/FortniteGame.FortPlayerPawn) "root" "SnapToTarget" "SnapToTarget" "SnapToTarget" true)
  (bind _returnvalue (Utilities|Time|SetTimerbyEvent (EventDispatchers|CreateEvent :!event Custom|IsSlidingCheck) 0.01 true true))
  (Variables|Default|SetZValueTickTimer _returnvalue)
  ;[0.2611 1 0.9568 0.5] WIND FORCE
  (Utilities|EffectsQualityCheck 0
    (:Cinematic
      _)
    (:Epic
      _))
  (Utilities|IsValid (Variables|Default|GetPhysical%20Wash%20Force)
    (:Is%20Not%20Valid
      (bind _returnvalue_2 (Game|SpawnActorfromClass MyTarget "/Game/Effects/Niagara/Blueprints/PhsyicsFields/FN_WashForce.FN_WashForce_C" (Transformation|GetActorTransform MyTarget) "Undefined" "MultiplyWithRoot" 0 2.5e+02 (Variables|Default|GetStrength) -999.0 0))
      (bind _output_get_2 (Variables|Default|SetPhysical%20Wash%20Force _returnvalue_2))
      (Transformation|AttachActorToComponent _output_get_2 (Variables|/Script/Engine.Actor|GetRootComponent MyTarget) "Root" "SnapToTarget" "SnapToTarget" "SnapToTarget" true)
      (bind (_direction _strength) (Timeline|Timeline ReverseFromEnd)
        (:Update
          (Variables|/Game/Effects/Niagara/Blueprints/PhsyicsFields/FN_WashForce.FN_WashForce_C|SetStrength _strength (Variables|Default|GetPhysical%20Wash%20Force)))))) ; Don't spawn the Rotor Wash multiple times
)

(event EventOnRemoval(EitherVFX-preferthisversion)
  (params (MyTarget :type Actor%20Object%20Reference) (Parameters :type Gameplay%20Cue%20Parameters%20Structure%20%28by%20ref%29) (ParticleComponents :type Array%20of%20FXSystem%20Component%20Object%20References) (AudioComponents :type Array%20of%20Audio%20Component%20Object%20References) (BurstCameraShakeInstance :type Camera%20Shake%20Base%20Object%20Reference) (BurstDecalInstance :type Decal%20Actor%20Object%20Reference))
  (|CallParentFunction MyTarget Parameters ParticleComponents AudioComponents BurstCameraShakeInstance BurstDecalInstance)
  (SoundIndicator|StopSoundIndicatorLooping (Variables|Default|GetFortSoundIndicator))
  (Utilities|Time|ClearandInvalidateTimerbyHandle 0 (Variables|Default|GetZValueTickTimer))
  (Actor|DestroyActor (Variables|Default|GetForce)))

(event Custom|ZValueTick
  (bind _as_fort_player_pawn (Variables|Default|GetAs%20Fort%20Player%20Pawn))
  (bind _mesh (Variables|/Script/Engine.Character|GetMesh _as_fort_player_pawn))
  (bind _returnvalue_location (Math|Transform|BreakTransform (Transformation|GetBoneTransform _mesh "foot_l_heel" "RTS_World")))
  (bind _returnvalue_location_2 (Math|Transform|BreakTransform (Transformation|GetBoneTransform _mesh "foot_r_toe_tip" "RTS_World")))
  (bind _particlecomps (Variables|Default|GetParticleComps))
  (bind _output (Utilities|Array|Get(acopy) _particlecomps))
  ; Disabled Ray Cast
  (Effects|Components|ParticleSystem|SetVectorParameter _output "LeftHeelPosition" _returnvalue_location)
  (Effects|Components|ParticleSystem|SetVectorParameter _output "RightToePosition" _returnvalue_location_2)
  (bind _outhit (Collision|LineTraceByChannel :WorldContextObject (Variables|Default|GetAs%20Fort%20Player%20Pawn) :Start (+ _returnvalue_location "0.000000,0.000000,10.000000") :End (- _returnvalue_location "0.000000,0.000000,20.000000") :TraceChannel "TraceTypeQuery8" :bTraceComplex false :DrawDebugType "Persistent" :bIgnoreSelf true :TraceColor "(R=1.000000,G=0.000000,B=0.000000,A=1.000000)" :TraceHitColor "(R=0.000000,G=1.000000,B=0.000000,A=1.000000)" :DrawTime 5.0))
  (bind (_outhit_bblockinghit _outhit_binitialoverlap _outhit_time _outhit_distance _outhit_location) (Collision|BreakHitResult _outhit))
  (bind _outhit_2 (Collision|LineTraceByChannel :WorldContextObject (Variables|Default|GetAs%20Fort%20Player%20Pawn) :Start (+ _returnvalue_location_2 "0.000000,0.000000,10.000000") :End (- _returnvalue_location_2 "0.000000,0.000000,20.000000") :TraceChannel "TraceTypeQuery8" :bTraceComplex false :DrawDebugType "Persistent" :bIgnoreSelf true :TraceColor "(R=1.000000,G=0.000000,B=0.000000,A=1.000000)" :TraceHitColor "(R=0.000000,G=1.000000,B=0.000000,A=1.000000)" :DrawTime 5.0))
  (bind (_outhit_bblockinghit_2 _outhit_binitialoverlap_2 _outhit_time_2 _outhit_distance_2 _outhit_location_2) (Collision|BreakHitResult _outhit_2))
  (Rendering|Debug|DrawDebugSphere 0 _outhit_location 1e+01 5 "(R=0.015571,G=0.018346,B=0.427083,A=1.000000)" 4e+01 1.0)
  (Rendering|Debug|DrawDebugSphere 0 _outhit_location_2 1e+01 5 "(R=0.048340,G=0.343750,B=0.054441,A=1.000000)" 4e+01 1.0))

(fn GetAffiliationsToShowFor
  (return (Utilities|Array|MakeArray "Friendly" "Neutral" "Hostile")))

(fn PlaySoundIndicator
  (params (CurrentTarget :type Actor%20Object%20Reference))
  (bind _output_get (Variables|Default|SetCurrentTarget CurrentTarget))
  (Utilities|IsValid _output_get
    (:Is%20Valid
      (if (not (Team|OnLocalPlayersTeam (Variables|Default|GetCurrentTarget) (Variables|Default|GetCurrentTarget)))
        (SoundIndicator|ShowSoundIndicator (Variables|Default|GetFortSoundIndicator) (Transformation|GetActorLocation (Variables|Default|GetCurrentTarget)) 0 1.0 0)))))

(fn ShouldShowSoundIndicator
  (params (PlayerController :type Fort%20Player%20Controller%20Object%20Reference))
  (bind _currenttarget (Variables|Default|GetCurrentTarget))
  (Utilities|IsValid _currenttarget
    (:Is%20Valid
      (return (!= PlayerController (Fortnite|GetControllerfromActor _currenttarget))))
    (:Is%20Not%20Valid
      (return true))))

(fn ConstructionScript)
