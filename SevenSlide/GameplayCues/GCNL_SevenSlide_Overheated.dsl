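; Looping gameplay cue for the SevenSlide overheated state, driving cooldown VFX and audio on a Fort player pawn.
; On loop start it attaches the first particle and audio components to the pawn's ExtraPart mesh root and debug-prints audio component names.
; A repeating timer feeds the Fuel attribute into the audio's FuelLevel parameter; removal or EndPlay clears it and fades the audio.
; Sound indicators cover all affiliations but exclude the target's own controller.
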(event EventOnLoopingStart(EitherVFX-preferthisversion)
  (params (MyTarget :type Actor%20Object%20Reference) (Parameters :type Gameplay%20Cue%20Parameters%20Structure%20%28by%20ref%29) (ParticleComponents :type Array%20of%20FXSystem%20Component%20Object%20References) (AudioComponents :type Array%20of%20Audio%20Component%20Object%20References))
  (|CallParentFunction MyTarget Parameters ParticleComponents AudioComponents)
  (bind _asfort_player_pawn (Utilities|Casting|CastToFortPlayerPawn MyTarget))
  (Variables|Default|SetAs%20Fort%20Player%20Pawn _asfort_player_pawn)
  (Transformation|AttachComponentToComponent (Utilities|Array|Get(acopy) ParticleComponents) (Customization|GetSkeletalMeshforPartType (Variables|Default|GetAs%20Fort%20Player%20Pawn) "ExtraPart" :!class /Script/FortniteGame.FortPlayerPawn) "root" "SnapToTarget" "SnapToTarget" "SnapToTarget" true)
  (for _array_element AudioComponents
    (Development|PrintString 0 (Utilities|GetDisplayName _array_element)))
  (bind _output_get (Variables|Default|SetCoolDownAudio (Utilities|Array|Get(acopy) AudioComponents)))
  (Utilities|IsValid _output_get
    (:Is%20Valid
      (Transformation|AttachComponentToComponent _output_get (Customization|GetSkeletalMeshforPartType (Variables|Default|GetAs%20Fort%20Player%20Pawn) "ExtraPart" :!class /Script/FortniteGame.FortPlayerPawn) "root" "SnapToTarget" "SnapToTarget" "SnapToTarget" true))
    (:Is%20Not%20Valid
      _))
  (bind _returnvalue (Utilities|Time|SetTimerbyEvent (EventDispatchers|CreateEvent :!event Custom|UpdateFuelAudio) 0.5 true true))
  (Variables|Default|SetCoolDownUpdateTimer _returnvalue))

(event Custom|UpdateFuelAudio
  (bind _as_fort_player_pawn (Variables|Default|GetAs%20Fort%20Player%20Pawn)
    (:then
      (Audio|Parameter|SetFloatParameter (Variables|Default|GetCoolDownAudio) "FuelLevel" (.ReturnValue (Ability|Attribute|GetFloatAttribute _as_fort_player_pawn "(AttributeName=\"Fuel\",Attribute=/Script/SevenSlideRuntime.SevenSlideAttributeSet:Fuel,AttributeOwner=\"/Script/CoreUObject.Class'/Script/SevenSlideRuntime.SevenSlideAttributeSet'\")")) :!class /Script/AudioExtensions.AudioParameterControllerInterface))
    (:else)))

(event EventOnRemoval(EitherVFX-preferthisversion)
  (params (MyTarget :type Actor%20Object%20Reference) (Parameters :type Gameplay%20Cue%20Parameters%20Structure%20%28by%20ref%29) (ParticleComponents :type Array%20of%20FXSystem%20Component%20Object%20References) (AudioComponents :type Array%20of%20Audio%20Component%20Object%20References) (BurstCameraShakeInstance :type Camera%20Shake%20Base%20Object%20Reference) (BurstDecalInstance :type Decal%20Actor%20Object%20Reference))
  (|CallParentFunction MyTarget Parameters ParticleComponents AudioComponents BurstCameraShakeInstance BurstDecalInstance)
  (Utilities|Time|ClearandInvalidateTimerbyHandle 0 (Variables|Default|GetCoolDownUpdateTimer))
  (bind _cooldownaudio (Variables|Default|GetCoolDownAudio)
    (:then
      (Audio|Components|Audio|FadeOut _cooldownaudio 0.5 0.0 "Linear" :!class /Script/Engine.AudioComponent))
    (:else)))

(event EventEndPlay
  (params (EndPlayReason :type EEndPlayReason%20Enum :default "Destroyed"))
  (|CallParentFunction EndPlayReason)
  (Utilities|Time|ClearandInvalidateTimerbyHandle 0 (Variables|Default|GetCoolDownUpdateTimer))
  (bind _cooldownaudio (Variables|Default|GetCoolDownAudio)
    (:then
      (Audio|Components|Audio|FadeOut _cooldownaudio 0.5 0.0 "Linear" :!class /Script/Engine.AudioComponent))
    (:else)))

(fn GetAffiliationsToShowFor
  (return (Utilities|Array|MakeArray "Friendly" "Neutral" "Hostile")))

(fn ShouldShowSoundIndicator
  (params (PlayerController :type Fort%20Player%20Controller%20Object%20Reference))
  (bind _currenttarget (Variables|Default|GetCurrentTarget))
  (Utilities|IsValid _currenttarget
    (:Is%20Valid
      (return (!= PlayerController (Fortnite|GetControllerfromActor _currenttarget))))
    (:Is%20Not%20Valid
      (return true))))

(fn ConstructionScript)
