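; Looping Gameplay Cue Notify for the SevenSlide warmup, adding a physics-field force and sound indicator rules on top of the parent cue.
; On looping start, where cosmetic events can run, it checks effects quality and the r.PhysicsField.EnableField cvar, then spawns FN_WashForce at the target's transform.
; On removal it destroys that stored force actor.
; The sound indicator covers friendly, neutral and hostile affiliations but is hidden from the current target's own controller.
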(event EventOnLoopingStart(EitherVFX-preferthisversion)
  (params (MyTarget :type Actor%20Object%20Reference) (Parameters :type Gameplay%20Cue%20Parameters%20Structure%20%28by%20ref%29) (ParticleComponents :type Array%20of%20FXSystem%20Component%20Object%20References) (AudioComponents :type Array%20of%20Audio%20Component%20Object%20References))
  (|CallParentFunction MyTarget Parameters ParticleComponents AudioComponents)
  (Networking|CanExecuteCosmeticEvents
    (:True
      (Utilities|EffectsQualityCheck 0
        (:Cinematic
          _)
        (:Epic
          _)
        (:High
          _))
      (bind _returnvalue (Development|GetConsoleVariableIntValue "r.PhysicsField.EnableField"))
      (if (== _returnvalue 1)
        (bind _returnvalue_2 (Game|SpawnActorfromClass 0 "/Game/Effects/Niagara/Blueprints/PhsyicsFields/FN_WashForce.FN_WashForce_C" (Transformation|GetActorTransform MyTarget) "Undefined" "MultiplyWithRoot" 0 1e+03 1.3 -999.0 0))
        (Variables|Default|SetForce _returnvalue_2)))))

(event EventOnRemoval(EitherVFX-preferthisversion)
  (params (MyTarget :type Actor%20Object%20Reference) (Parameters :type Gameplay%20Cue%20Parameters%20Structure%20%28by%20ref%29) (ParticleComponents :type Array%20of%20FXSystem%20Component%20Object%20References) (AudioComponents :type Array%20of%20Audio%20Component%20Object%20References) (BurstCameraShakeInstance :type Camera%20Shake%20Base%20Object%20Reference) (BurstDecalInstance :type Decal%20Actor%20Object%20Reference))
  (|CallParentFunction MyTarget Parameters ParticleComponents AudioComponents BurstCameraShakeInstance BurstDecalInstance)
  (Utilities|IsValid (Variables|Default|GetForce)
    (:Is%20Valid
      (Actor|DestroyActor (Variables|Default|GetForce)))))

(fn GetAffiliationsToShowFor
  (return (Utilities|Array|MakeArray "Friendly" "Neutral" "Hostile")))

(fn ShouldShowSoundIndicator
  (params (PlayerController :type Fort%20Player%20Controller%20Object%20Reference))
  (bind _currenttarget (Variables|Default|GetCurrentTarget))
  (Utilities|IsValid _currenttarget
    (:Is%20Valid
      (return (!= PlayerController (Fortnite|GetControllerfromActor _currenttarget))))
    (:Is%20Not%20Valid
      (return true))))

(fn ConstructionScript)
