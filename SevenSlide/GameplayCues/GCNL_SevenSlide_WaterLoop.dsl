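; Looping GameplayCue for the SevenSlide water effect that adds a physical wash force and sound indicator to the target.
; On loop start, with an effects-quality check, it spawns FN_WashForce once, attaches it to the target's root and drives its Strength from a reversed timeline.
; Removal destroys the actor in a separate Force variable, not Physical Wash Force.
; The indicator uses the first application-burst sound and excludes the target's own controller.
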(event EventOnLoopingStart(Niagara)
  (params (MyTarget :type Actor%20Object%20Reference) (Parameters :type Gameplay%20Cue%20Parameters%20Structure%20%28by%20ref%29) (ParticleComponents :type Array%20of%20Niagara%20Particle%20System%20Component%20Object%20References) (AudioComponents :type Array%20of%20Audio%20Component%20Object%20References))
  (|CallParentFunction MyTarget Parameters ParticleComponents AudioComponents)
  ;[0.2611 1 0.9568 0.5] WIND FORCE
  (Utilities|EffectsQualityCheck 0
    (:Cinematic
      _)
    (:Epic
      _))
  (Utilities|IsValid (Variables|Default|GetPhysical%20Wash%20Force)
    (:Is%20Not%20Valid
      (bind _returnvalue (Game|SpawnActorfromClass MyTarget "/Game/Effects/Niagara/Blueprints/PhsyicsFields/FN_WashForce.FN_WashForce_C" (Transformation|GetActorTransform MyTarget) "Undefined" "MultiplyWithRoot" 0 2.5e+02 (Variables|Default|GetStrength) -999.0 0))
      (bind _output_get (Variables|Default|SetPhysical%20Wash%20Force _returnvalue))
      (Transformation|AttachActorToComponent _output_get (Variables|/Script/Engine.Actor|GetRootComponent MyTarget) "Root" "SnapToTarget" "SnapToTarget" "SnapToTarget" true)
      (bind (_direction _strength) (Timeline|Timeline ReverseFromEnd)
        (:Update
          (Variables|/Game/Effects/Niagara/Blueprints/PhsyicsFields/FN_WashForce.FN_WashForce_C|SetStrength _strength (Variables|Default|GetPhysical%20Wash%20Force)))))) ; Don't spawn the Rotor Wash multiple times
)

(event EventOnRemoval(EitherVFX-preferthisversion)
  (params (MyTarget :type Actor%20Object%20Reference) (Parameters :type Gameplay%20Cue%20Parameters%20Structure%20%28by%20ref%29) (ParticleComponents :type Array%20of%20FXSystem%20Component%20Object%20References) (AudioComponents :type Array%20of%20Audio%20Component%20Object%20References) (BurstCameraShakeInstance :type Camera%20Shake%20Base%20Object%20Reference) (BurstDecalInstance :type Decal%20Actor%20Object%20Reference))
  (|CallParentFunction MyTarget Parameters ParticleComponents AudioComponents BurstCameraShakeInstance BurstDecalInstance)
  (Actor|DestroyActor (Variables|Default|GetForce)))

(fn GetAffiliationsToShowFor
  (return (Utilities|Array|MakeArray "Friendly" "Neutral" "Hostile")))

(fn PlaySoundIndicator
  (params (CurrentTarget :type Actor%20Object%20Reference))
  (bind _applicationburst (Variables|ApplicationEffects|GetApplicationBurst))
  (bind (_burstparticles _burstsounds) (Utilities|Struct|BreakFortBurstEffectData _applicationburst))
  (bind (_condition _attachment _soundcue) (Utilities|Struct|BreakFortGameplayCueAudioInfo (Utilities|Array|Get(aref) _burstsounds)))
  (bind _output_get (Variables|Default|SetCurrentTarget CurrentTarget))
  (Utilities|IsValid _output_get
    (:Is%20Valid
      (if (> (Utilities|Array|Length _burstsounds) 0)
        (SoundIndicator|ShowSoundIndicator (Variables|Default|GetFortSoundIndicator) (Transformation|GetActorLocation (Variables|Default|GetCurrentTarget)) _soundcue 1.0 0)))))

(fn ShouldShowSoundIndicator
  (params (PlayerController :type Fort%20Player%20Controller%20Object%20Reference))
  (bind _currenttarget (Variables|Default|GetCurrentTarget))
  (Utilities|IsValid _currenttarget
    (:Is%20Valid
      (return (!= PlayerController (Fortnite|GetControllerfromActor _currenttarget))))
    (:Is%20Not%20Valid
      (return true))))

(fn ConstructionScript)
